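# Animation Blueprint for the SeaFlake backpack cosmetic, adding secondary motion to its sushi, cloth, eye and rope bones.
# Chained spring-driven Anim Dynamics nodes simulate each part, with dyn_sushi_b getting a second, linear-only pass for vertical bounce.
# A final additive Modify Bone offsets and rotates the pack bone.
# The update event sets Gravity Override through a deprecated function, but no AnimGraph node reads that variable.
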AnimBlueprint /BRCosmetics/Accessories/FORT_Backpacks/Backpack_M_MED_SeaFlake/Meshes/M_MED_SeaFlake_Pack_AnimBP
target_skeleton=&/M_MED_SeaFlake_Pack_Skeleton
graphs [
 AnimGraph {nodes [
   n1 = Anim Dynamics - Bone: dyn_sushi_b(ComponentPose=Local To Component(LocalPose=Input Pose()), Alpha=1, ExternalForce="0,0,40") {
    angularDampingOverride=0.9999
    angularSpringConstant=2000
    bAngularSpring=true
    bLinearSpring=true
    bOverrideAngularDamping=true
    bOverrideLinearDamping=true
    gravityScale=0.5
    linearDampingOverride=0.999
    linearSpringConstant=180
    alphaBoolBlend {blendOption=Linear}
    alphaScaleBiasClamp {bClampResult=true, clampMin=0.5}
    boundBone {boneName=dyn_sushi_b}
    componentAppliedLinearAccClamp=<300 300 100>
    componentLinearAccScale=<0.3 0.3 0.3>
    componentLinearVelScale=<0.2 0.2 0.2>
    gravityOverride {z=-500}
    physicsBodyDefinitions [{
      boundBone {boneName=dyn_sushi_b}
      boxExtents=<15 15 15>
      constraintSetup {angularLimitsMax=<5 8 5>, angularLimitsMin=<-0.5 0 0>, angularTarget=<1 0 0>, linearAxesMax=<2 0.5 0.5>, linearAxesMin=<-0.5 0 0>}
      localJointOffset=<5 0 0>
     }]
   }
   n2 = Anim Dynamics - Bone: dyn_sushi_a(ComponentPose=@n1, Alpha=1, ExternalForce="0,0,40") {
    angularDampingOverride=0.999
    angularSpringConstant=2000
    bAngularSpring=true
    bLinearSpring=true
    bOverrideAngularDamping=true
    bOverrideLinearDamping=true
    gravityScale=0.5
    linearDampingOverride=0.999
    linearSpringConstant=200
    numSolverIterationsPreUpdate=3
    alphaBoolBlend {blendOption=Linear}
    alphaScaleBiasClamp {bClampResult=true, clampMin=0.2}
    boundBone {boneName=dyn_sushi_a}
    componentAppliedLinearAccClamp=<450 450 100>
    componentLinearAccScale=<0.2 0.2 0.2>
    componentLinearVelScale=<0.1 0.1 0.1>
    gravityOverride {z=-500}
    physicsBodyDefinitions [{
      boundBone {boneName=dyn_sushi_a}
      boxExtents=<20 20 20>
      constraintSetup {angularLimitsMax=<5 0 5>, angularLimitsMin=<-0.5 -5 0>, angularTarget=<1 0 0>, linearAxesMax=<2 0.5 0.5>, linearAxesMin=<-0.5 0 0>}
      localJointOffset=<2 0 0>
     }]
   }
   n3 = Anim Dynamics - Bone: dyn_sushi_base(ComponentPose=@n2, Alpha=1, ExternalForce="0,0,40") {
    angularDampingOverride=0.999
    angularSpringConstant=2000
    bAngularSpring=true
    bLinearSpring=true
    bOverrideAngularDamping=true
    bOverrideLinearDamping=true
    gravityScale=0.5
    linearDampingOverride=0.9999
    linearSpringConstant=80
    alphaBoolBlend {blendOption=Linear}
    alphaScaleBiasClamp {bClampResult=true}
    boundBone {boneName=dyn_sushi_base}
    componentAppliedLinearAccClamp=<175 175 70>
    componentLinearAccScale=<0.4 0.3 0.4>
    componentLinearVelScale=<0.3 0.4 0.3>
    gravityOverride {z=-500}
    physicsBodyDefinitions [{
      boundBone {boneName=dyn_sushi_base}
      boxExtents=<25 25 25>
      constraintSetup {angularLimitsMax=<3 0 3>, angularLimitsMin=<-0.5 0 0>, angularTarget=<1 0 0>, linearAxesMax=<2 1 0.5>, linearAxesMin=<-0.5 0 0>}
      localJointOffset=<2 0 0>
     }]
   }
   n4 = Anim Dynamics - Bone: dyn_sushi_b(ComponentPose=@n3, Alpha=1, ExternalForce="0,0,0") {
    bLinearSpring=true
    bOverrideLinearDamping=true
    bUseGravityOverride=true
    linearDampingOverride=0.99
    linearSpringConstant=5
    alphaBoolBlend {blendOption=Linear}
    alphaScaleBiasClamp {bClampResult=true, clampMax=0.6}
    boundBone {boneName=dyn_sushi_b}
    componentAppliedLinearAccClamp=<0 0 1000>
    componentLinearAccScale {z=0.5}
    componentLinearVelScale {z=0.6}
    gravityOverride {z=-960}
    physicsBodyDefinitions [{
      boundBone {boneName=dyn_sushi_b}
      boxExtents=<5 10 10>
      constraintSetup {angularLimitsMax=<0 0 0>, angularLimitsMin=<0 0 0>, angularTarget=<0 0 0>, linearAxesMax=<3 0 0>}
      localJointOffset=<0.5 0 0>
     }]
   }
   n5 = Anim Dynamics - Bone: dyn_cloth_01!(ComponentPose=@n4, Alpha=1, GravityOverride:Vector="0,0,0", ExternalForce="0,0,0") -> (Pose) {
    angularDampingOverride=0.999
    angularSpringConstant=100
    bAngularSpring=true
    bOverrideAngularDamping=true
    bUseGravityOverride=true
    alphaBoolBlend {blendOption=Linear}
    alphaScaleBiasClamp {bClampResult=true, clampMin=1}
    boundBone {boneName=dyn_cloth_01}
    componentAppliedLinearAccClamp=<1500 1500 1000>
    componentLinearAccScale=<0.7 0.7 0.7>
    componentLinearVelScale=<0.7 0.7 0.7>
    physicsBodyDefinitions [{
      boundBone {boneName=dyn_cloth_01}
      boxExtents=<10 10 10>
      constraintSetup {angularLimitsMax=<0 0 10>, angularLimitsMin=<0 -120 -10>, angularTarget=<1 0 0>}
      localJointOffset=<2 0 0>
     }]
   }
   n6 = Anim Dynamics - Bone: dyn_cloth_02!(ComponentPose=@n5, Alpha=1, GravityOverride:Vector="0,0,0", ExternalForce="0,0,0") -> (Pose) {
    angularDampingOverride=0.998
    angularSpringConstant=80
    bAngularSpring=true
    bOverrideAngularDamping=true
    bUseGravityOverride=true
    alphaBoolBlend {blendOption=Linear}
    alphaScaleBiasClamp {bClampResult=true, clampMin=1}
    boundBone {boneName=dyn_cloth_02}
    componentAppliedLinearAccClamp=<1500 1500 1000>
    componentLinearAccScale=<0.5 0.5 0.5>
    componentLinearVelScale=<0.7 0.7 0.7>
    physicsBodyDefinitions [{
      boundBone {boneName=dyn_cloth_02}
      boxExtents=<8 8 8>
      constraintSetup {angularLimitsMax=<0 20 5>, angularLimitsMin=<0 -70 -5>, angularTarget=<1 0 0>}
      localJointOffset=<1.5 0 0>
     }]
   }
   n7 = Anim Dynamics - Bone: dyn_cloth_03!(ComponentPose=@n6, Alpha=1, GravityOverride:Vector="0,0,0", ExternalForce="0,0,0") -> (Pose) {
    angularDampingOverride=0.999
    angularSpringConstant=60
    bAngularSpring=true
    bOverrideAngularDamping=true
    bUseGravityOverride=true
    alphaBoolBlend {blendOption=Linear}
    alphaScaleBiasClamp {bClampResult=true, clampMin=1}
    boundBone {boneName=dyn_cloth_03}
    componentAppliedLinearAccClamp=<1500 1500 1000>
    componentLinearAccScale=<0.4 0.4 0.4>
    componentLinearVelScale=<0.5 0.5 0.5>
    physicsBodyDefinitions [{
      boundBone {boneName=dyn_cloth_03}
      boxExtents=<6 6 6>
      constraintSetup {angularLimitsMax=<0 30 5>, angularLimitsMin=<0 -60 -5>, angularTarget=<1 0 0>}
      localJointOffset=<1 0 0>
     }]
   }
   n8 = Anim Dynamics - Bone: dyn_eye_02_b(ComponentPose=@n7, Alpha=1, ExternalForce="0,0,0") {
    angularDampingOverride=1
    angularSpringConstant=3000
    bAngularSpring=true
    bGravityOverrideInSimSpace=true
    bOverrideAngularDamping=true
    gravityScale=0
    alphaBoolBlend {blendOption=Linear}
    alphaScaleBiasClamp {bClampResult=true}
    boundBone {boneName=dyn_eye_02_b}
    componentAppliedLinearAccClamp=<2000 2000 2500>
    componentLinearAccScale=<0.2 0.4 0.7>
    componentLinearVelScale=<0.5 0.4 0.8>
    physicsBodyDefinitions [{
      boundBone {boneName=dyn_eye_02_b}
      boxExtents=<12 12 12>
      constraintSetup {angularLimitsMax=<35 40 35>, angularLimitsMin=<-35 -40 -35>, angularTarget=<1 0 0>}
      localJointOffset=<1 1.5 1>
     }]
   }
   n9 = Anim Dynamics - Bone: dyn_eye_01_a(ComponentPose=@n8, Alpha=1, ExternalForce="0,0,0") {
    angularDampingOverride=1
    angularSpringConstant=2000
    bAngularSpring=true
    bGravityOverrideInSimSpace=true
    bOverrideAngularDamping=true
    gravityScale=0
    alphaBoolBlend {blendOption=Linear}
    alphaScaleBiasClamp {bClampResult=true}
    boundBone {boneName=dyn_eye_01_a}
    componentAppliedLinearAccClamp=<1500 1500 2000>
    componentLinearAccScale=<0.4 0.4 0.4>
    componentLinearVelScale=<0.5 0.5 0.5>
    physicsBodyDefinitions [{
      boundBone {boneName=dyn_eye_01_a}
      boxExtents=<12 12 8>
      constraintSetup {angularLimitsMax=<35 40 35>, angularLimitsMin=<-35 -40 -35>, angularTarget=<1 0 0>}
      localJointOffset=<1 1.5 1>
     }]
   }
   n10 = Anim Dynamics - Bone: dyn_rope_01!(ComponentPose=@n9, Alpha=0.7, GravityOverride:Vector="0,0,-980") -> (Pose) {
    angularDampingOverride=0.999
    angularSpringConstant=1000
    bAngularSpring=true
    bOverrideAngularDamping=true
    bUseGravityOverride=true
    alphaBoolBlend {blendOption=Linear}
    alphaScaleBiasClamp {bClampResult=true}
    boundBone {boneName=dyn_rope_01}
    componentAppliedLinearAccClamp=<1000 1000 1000>
    componentLinearAccScale=<0.4 0.4 0.4>
    componentLinearVelScale=<0.5 0.5 0.5>
    physicsBodyDefinitions [{
      boundBone {boneName=dyn_rope_01}
      boxExtents=<6 6 6>
      constraintSetup {angularLimitsMax=<0 25 25>, angularLimitsMin=<0 -25 -25>, angularTarget=<1 0 0>, linearAxesMin=<-1.5 0 0>}
      localJointOffset=<1 0 0>
     }]
   }
   n11 = "Transform (Modify) Bone - Bone: pack"(ComponentPose=@n10, Translation="0,2.8,0", Rotation="0,6.5,-10") {
    rotationMode=BMM_Additive
    rotationSpace=BCS_BoneSpace
    translationMode=BMM_Additive
    translationSpace=BCS_BoneSpace
    alphaBoolBlend {blendInTime=0.1, blendOption=Linear, blendOutTime=0.1}
    boneToModify {boneName=pack}
    rotation {roll=-10, yaw=6.5}
    translation {y=2.8}
   }
   Output Pose(Result=Component To Local(ComponentPose=@n11)) {layerGroup=None}
  ]}
 EventGraph {nodes [
   Event BlueprintUpdateAnimation()
   > n1 = Gravity Override State Function DEPRECATED(CharacterPartAnimInstance=Self-Reference(), GameTimeInSeconds=GetGameTimeInSeconds(), GravityOverrideParams=Get Gravity Override Params())
   > Set Gravity Override(Gravity Override=@n1)
   TryGetPawnOwner()
  ]}
]
variables [
 Gravity Override {type=Vector}
 Gravity Override Params {type=GravityOverrideParamsStruct}
]
